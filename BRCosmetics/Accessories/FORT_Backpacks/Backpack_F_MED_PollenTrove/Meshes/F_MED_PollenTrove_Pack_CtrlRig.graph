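# Control Rig for the PollenTrove backpack that procedurally moves three bee body bones.
# The BeeMovement vector is remapped to per-axis offsets and spring-interpolated, with damping varying per bee so their lag differs.
# Each bee then aims at shared BeeFwd and BeeSide controls.
# A final pass adds spring-smoothed random Y/Z hover jitter with per-bee timing; wing bones are untouched and a TimeOffset node is unconnected.
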ControlRigBlueprint /BRCosmetics/Accessories/FORT_Backpacks/Backpack_F_MED_PollenTrove/Meshes/F_MED_PollenTrove_Pack_CtrlRig
graphs [
 RigVMFunctionLibrary {links [], nodes []}
 RigVMModel {nodes [
   RigUnit_BeginExecution = &RigUnit_BeginExecution()
   n1 = &RigUnit_GetTransform(Item=(Type=Bone,Name="dyn_bee_01_body"), bInitial=False)
   n2 = RigVMVariableNode() {variable=BeeMovement}
   Remap = &RigVMFunction_MathDoubleRemap(
    SourceMaximum=1
    SourceMinimum=-1
    TargetMaximum=-8
    TargetMinimum=8
    Value=&RigVMFunction_MathDoubleMul(B=-1, A=@n2.Value.X).Result
   )
   n3 = /Script/ControlRig.RigUnit_SpringInterpVectorV2(
    CriticalDamping=0.5
    Force=(X=0,Y=0,Z=0)
    Strength=1
    TargetVelocityAmount=0
    bInitializeFromTarget=False
    bUseCurrentInput=False
    Target=/Script/RigVM.RigVMFunction_MathVectorAdd(
     B=(X=0,Y=0,Z=0)
     A=@n1.Transform.Translation
     B.X=@Remap.Result
     B.Y=&RigVMFunction_MathDoubleRemap(SourceMaximum=1, SourceMinimum=-1, TargetMaximum=-10, TargetMinimum=1, Value=@n2.Value.Y).Result
     B.Z=&RigVMFunction_MathDoubleRemap(SourceMaximum=1, SourceMinimum=-1, TargetMaximum=-2, TargetMinimum=4, Value=@n2.Value.Z).Result
    ) {is=struct RigVMUnitNode, template="Add::Execute(in A,in B,out Result)"}.Result
    Current=@n1.Transform.Translation
   ) {
    is=struct RigVMUnitNode
    template="SpringInterp::Execute(in Target,in Strength,in CriticalDamping,in Force,in bUseCurrentInput,in Current,in TargetVelocityAmount,in bInitializeFromTarget,out Result,out Velocity)"
   }
   &RigUnit_ModifyTransforms(
    ItemToModify=((Item=(Type=Bone,Name="dyn_bee_01_body"),Transform=(Rotation=(X=0,Y=0,Z=0,W=1),Translation=(X=0,Y=0,Z=0),Scale3D=(X=1,Y=1,Z=1))))
    Mode=OverrideLocal
    Weight=1
    WeightMinimum=0
    ItemToModify.0.Transform.Rotation=@n1.Transform.Rotation
    ItemToModify.0.Transform.Scale3D=@n1.Transform.Scale3D
    ExecutePin=@RigUnit_BeginExecution.ExecutePin
    ItemToModify.0.Transform.Translation=@n3.Result
   )
   n4 = &RigUnit_GetTransform(Item=(Type=Bone,Name="dyn_bee_02_body"), bInitial=False)
   n5 = RigVMVariableNode() {variable=BeeMovement}
   n6 = &RigVMFunction_MathDoubleMul(B=-1, A=@n5.Value.X)
   n7 = &RigUnit_SpringInterpVectorV2(
    CriticalDamping=0.4
    Force=(X=0,Y=0,Z=0)
    Strength=0.8
    TargetVelocityAmount=0
    bInitializeFromTarget=False
    bUseCurrentInput=False
    Target=&RigVMFunction_MathVectorAdd(
     B=(X=0,Y=0,Z=0)
     A=@n4.Transform.Translation
     B.X=&RigVMFunction_MathDoubleRemap(SourceMaximum=1, SourceMinimum=-1, TargetMaximum=-8, TargetMinimum=8, Value=@n6.Result).Result
     B.Y=&RigVMFunction_MathDoubleRemap(SourceMaximum=1, SourceMinimum=-1, TargetMaximum=-10, TargetMinimum=1, Value=@n5.Value.Y).Result
     B.Z=&RigVMFunction_MathDoubleRemap(SourceMaximum=1, SourceMinimum=-1, TargetMaximum=-2, TargetMinimum=4, Value=@n5.Value.Z).Result
    ) {template="Add::Execute(in A,in B,out Result)"}.Result
    Current=@n4.Transform.Translation
   ) {template="SpringInterp::Execute(in Target,in Strength,in CriticalDamping,in Force,in bUseCurrentInput,in Current,in TargetVelocityAmount,in bInitializeFromTarget,out Result,out Velocity)"}
   > &RigUnit_ModifyTransforms(
    ItemToModify=((Item=(Type=Bone,Name="dyn_bee_02_body"),Transform=(Rotation=(X=0,Y=0,Z=0,W=1),Translation=(X=0,Y=0,Z=0),Scale3D=(X=1,Y=1,Z=1))))
    Mode=OverrideLocal
    Weight=1
    WeightMinimum=0
    ItemToModify.0.Transform.Rotation=@n4.Transform.Rotation
    ItemToModify.0.Transform.Scale3D=@n4.Transform.Scale3D
    ItemToModify.0.Transform.Translation=@n7.Result
   )
   n8 = &RigUnit_GetTransform(Item=(Type=Bone,Name="dyn_bee_03_body"), bInitial=False)
   n9 = RigVMVariableNode() {variable=BeeMovement}
   n10 = &RigVMFunction_MathDoubleMul(B=-1, A=@n9.Value.X)
   n11 = &RigUnit_SpringInterpVectorV2(
    CriticalDamping=0.3
    Force=(X=0,Y=0,Z=0)
    Strength=0.8
    TargetVelocityAmount=0
    bInitializeFromTarget=False
    bUseCurrentInput=False
    Target=&RigVMFunction_MathVectorAdd(
     B=(X=0,Y=0,Z=0)
     A=@n8.Transform.Translation
     B.X=&RigVMFunction_MathDoubleRemap(SourceMaximum=1, SourceMinimum=-1, TargetMaximum=-8, TargetMinimum=8, Value=@n10.Result).Result
     B.Y=&RigVMFunction_MathDoubleRemap(SourceMaximum=1, SourceMinimum=-1, TargetMaximum=-10, TargetMinimum=1, Value=@n9.Value.Y).Result
     B.Z=&RigVMFunction_MathDoubleRemap(SourceMaximum=1, SourceMinimum=-1, TargetMaximum=-2, TargetMinimum=4, Value=@n9.Value.Z).Result
    ) {template="Add::Execute(in A,in B,out Result)"}.Result
    Current=@n8.Transform.Translation
   ) {template="SpringInterp::Execute(in Target,in Strength,in CriticalDamping,in Force,in bUseCurrentInput,in Current,in TargetVelocityAmount,in bInitializeFromTarget,out Result,out Velocity)"}
   > &RigUnit_ModifyTransforms(
    ItemToModify=((Item=(Type=Bone,Name="dyn_bee_03_body"),Transform=(Rotation=(X=0,Y=0,Z=0,W=1),Translation=(X=0,Y=0,Z=0),Scale3D=(X=1,Y=1,Z=1))))
    Mode=OverrideLocal
    Weight=1
    WeightMinimum=0
    ItemToModify.0.Transform.Rotation=@n8.Transform.Rotation
    ItemToModify.0.Transform.Scale3D=@n8.Transform.Scale3D
    ItemToModify.0.Transform.Translation=@n11.Result
   )
   > /Script/ControlRig.RigUnit_AimItem(
    DebugSettings=(bEnabled=False,Scale=10,WorldOffset=(Rotation=(X=0,Y=0,Z=0,W=1),Translation=(X=0,Y=0,Z=0),Scale3D=(X=1,Y=1,Z=1)))
    Item=(Type=Bone,Name="dyn_bee_01_body")
    Primary=(Weight=1,Axis=(X=-1,Y=0,Z=0),Target=(X=1,Y=0,Z=0),Kind=Location,Space=(Type=Control,Name="BeeFwd_ctrl"))
    Secondary=(Weight=1,Axis=(X=0,Y=0,Z=1),Target=(X=1,Y=0,Z=0),Kind=Location,Space=(Type=Control,Name="BeeSide_ctrl"))
    Weight=1
   ) {is=struct RigVMUnitNode}
   > &RigUnit_AimItem(
    DebugSettings=(bEnabled=False,Scale=10,WorldOffset=(Rotation=(X=0,Y=0,Z=0,W=1),Translation=(X=0,Y=0,Z=0),Scale3D=(X=1,Y=1,Z=1)))
    Item=(Type=Bone,Name="dyn_bee_02_body")
    Primary=(Weight=1,Axis=(X=-1,Y=0,Z=0),Target=(X=1,Y=0,Z=0),Kind=Location,Space=(Type=Control,Name="BeeFwd_ctrl"))
    Secondary=(Weight=1,Axis=(X=0,Y=0,Z=1),Target=(X=1,Y=0,Z=0),Kind=Location,Space=(Type=Control,Name="BeeSide_ctrl"))
    Weight=1
   )
   > &RigUnit_AimItem(
    DebugSettings=(bEnabled=False,Scale=10,WorldOffset=(Rotation=(X=0,Y=0,Z=0,W=1),Translation=(X=0,Y=0,Z=0),Scale3D=(X=1,Y=1,Z=1)))
    Item=(Type=Bone,Name="dyn_bee_03_body")
    Primary=(Weight=1,Axis=(X=-1,Y=0,Z=0),Target=(X=1,Y=0,Z=0),Kind=Location,Space=(Type=Control,Name="BeeFwd_ctrl"))
    Secondary=(Weight=1,Axis=(X=0,Y=0,Z=1),Target=(X=1,Y=0,Z=0),Kind=Location,Space=(Type=Control,Name="BeeSide_ctrl"))
    Weight=1
   )
   SpringInterp = &RigUnit_SpringInterpVectorV2(
    CriticalDamping=0.8
    Current=(X=0,Y=0,Z=0)
    Force=(X=0,Y=1,Z=1)
    Strength=3
    Target=(X=0,Y=0,Z=0)
    TargetVelocityAmount=0
    bInitializeFromTarget=False
    bUseCurrentInput=False
    Target.Y=/Script/RigVM.RigVMFunction_RandomFloat(Duration=1, Maximum=2, Minimum=-0.1, Seed=217) {is=struct RigVMUnitNode, template="Random::Execute(in Seed,in Minimum,in Maximum,in Duration,out Result)"}.Result
    Target.Z=&RigVMFunction_RandomFloat(Duration=1, Maximum=2, Minimum=-2, Seed=641) {template="Random::Execute(in Seed,in Minimum,in Maximum,in Duration,out Result)"}.Result
   ) {template="SpringInterp::Execute(in Target,in Strength,in CriticalDamping,in Force,in bUseCurrentInput,in Current,in TargetVelocityAmount,in bInitializeFromTarget,out Result,out Velocity)"}
   > &RigUnit_ModifyTransforms(
    ItemToModify=((Item=(Type=Bone,Name="dyn_bee_01_body"),Transform=(Rotation=(X=0,Y=0,Z=0,W=1),Translation=(X=0,Y=0,Z=0),Scale3D=(X=1,Y=1,Z=1))))
    Weight=1
    WeightMinimum=0
    ItemToModify.0.Transform.Translation=@SpringInterp.Result
   )
   n12 = &RigUnit_SpringInterpVectorV2(
    CriticalDamping=0.8
    Current=(X=0,Y=0,Z=0)
    Force=(X=0,Y=1,Z=1)
    Strength=3
    Target=(X=0,Y=0,Z=0)
    TargetVelocityAmount=0
    bInitializeFromTarget=False
    bUseCurrentInput=False
    Target.Y=&RigVMFunction_RandomFloat(Duration=1.3, Maximum=2, Minimum=-2, Seed=650) {template="Random::Execute(in Seed,in Minimum,in Maximum,in Duration,out Result)"}.Result
    Target.Z=&RigVMFunction_RandomFloat(Duration=1.3, Maximum=1, Minimum=-1, Seed=230) {template="Random::Execute(in Seed,in Minimum,in Maximum,in Duration,out Result)"}.Result
   ) {template="SpringInterp::Execute(in Target,in Strength,in CriticalDamping,in Force,in bUseCurrentInput,in Current,in TargetVelocityAmount,in bInitializeFromTarget,out Result,out Velocity)"}
   > &RigUnit_ModifyTransforms(
    ItemToModify=((Item=(Type=Bone,Name="dyn_bee_02_body"),Transform=(Rotation=(X=0,Y=0,Z=0,W=1),Translation=(X=0,Y=0,Z=0),Scale3D=(X=1,Y=1,Z=1))))
    Weight=1
    WeightMinimum=0
    ItemToModify.0.Transform.Translation=@n12.Result
   )
   n13 = &RigUnit_SpringInterpVectorV2(
    CriticalDamping=0.8
    Current=(X=0,Y=0,Z=0)
    Force=(X=0,Y=1,Z=1)
    Strength=3
    Target=(X=0,Y=0,Z=0)
    TargetVelocityAmount=0
    bInitializeFromTarget=False
    bUseCurrentInput=False
    Target.Z=&RigVMFunction_RandomFloat(Duration=0.9, Maximum=2, Minimum=-1, Seed=650) {template="Random::Execute(in Seed,in Minimum,in Maximum,in Duration,out Result)"}.Result
    Target.Y=&RigVMFunction_RandomFloat(Duration=0.9, Maximum=0.1, Minimum=-1, Seed=230) {template="Random::Execute(in Seed,in Minimum,in Maximum,in Duration,out Result)"}.Result
   ) {template="SpringInterp::Execute(in Target,in Strength,in CriticalDamping,in Force,in bUseCurrentInput,in Current,in TargetVelocityAmount,in bInitializeFromTarget,out Result,out Velocity)"}
   > &RigUnit_ModifyTransforms(
    ItemToModify=((Item=(Type=Bone,Name="dyn_bee_03_body"),Transform=(Rotation=(X=0,Y=0,Z=0,W=1),Translation=(X=0,Y=0,Z=0),Scale3D=(X=1,Y=1,Z=1))))
    Weight=1
    WeightMinimum=0
    ItemToModify.0.Transform.Translation=@n13.Result
   )
   /Script/RigVM.RigVMFunction_TimeOffsetTransform(BufferSize=20, SecondsAgo=0.1, TimeRange=1, Value=(Rotation=(X=0,Y=0,Z=0,W=1),Translation=(X=0,Y=0,Z=0),Scale3D=(X=1,Y=1,Z=1))) {is=struct RigVMUnitNode, template="TimeOffset::Execute(in Value,in SecondsAgo,in BufferSize,in TimeRange,out Result)"}
  ]}
]
hierarchy [
 root {loc=<0 0 0>, rot=<0 0 0 1>}
 pack {parent=root, loc=<0 -0.7316 0>, rot=<0 0 0 1>}
 dyn_bee_01_body {parent=pack, loc=<2.6892 -7.7816 4.597>, rot=<-0.5 0.5 0.5 -0.5>}
 dyn_bee_01_wing_r {parent=dyn_bee_01_body, loc=<0.0028 1.3305 0.2923>, rot=<0.1391 0.519 -0.2183 -0.8147>}
 dyn_bee_01_wing_l {parent=dyn_bee_01_body, loc=<0.0028 1.3305 -0.2922>, rot=<-0.1391 -0.519 -0.2183 -0.8147>}
 dyn_bee_02_body {parent=pack, loc=<11.4102 -10.2313 2.1507>, rot=<-0.5 0.5 0.5 -0.5>}
 dyn_bee_02_wing_l {parent=dyn_bee_02_body, loc=<0.0028 1.3305 -0.2922>, rot=<-0.1391 -0.519 -0.2183 -0.8147>}
 dyn_bee_02_wing_r {parent=dyn_bee_02_body, loc=<0.0028 1.3305 0.2923>, rot=<0.1391 0.519 -0.2183 -0.8147>}
 dyn_bee_03_body {parent=pack, loc=<0.0002 -10.5416 -1.7924>, rot=<-0.5 0.5 0.5 -0.5>}
 dyn_bee_03_wing_l {parent=dyn_bee_03_body, loc=<0.0028 1.3305 -0.2922>, rot=<-0.1391 -0.519 -0.2183 -0.8147>}
 dyn_bee_03_wing_r {parent=dyn_bee_03_body, loc=<0.0028 1.3305 0.2923>, rot=<0.1391 0.519 -0.2183 -0.8147>}
 BeeFwd_ctrl {controlType=EULER_TRANSFORM, type=Control, loc=<0 0 0>, rot=<0 0 0 1>}
 BeeSide_ctrl {controlType=EULER_TRANSFORM, type=Control, loc=<0 0 0>, rot=<0 0 0 1>}
]
variables [BeeMovement {type=FVector, default=(X=0,Y=0,Z=0)}]
