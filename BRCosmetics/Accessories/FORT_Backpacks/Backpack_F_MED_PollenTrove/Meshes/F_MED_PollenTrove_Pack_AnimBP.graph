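# Animation Blueprint for the Pollen Trove backpack, animating the pack and three floating bees.
# The event graph derives pack offsets from spine rotation, bee hover from real-time sine waves, and player movement for a Control Rig.
# The AnimGraph runs that rig, then layers additive bone offsets on pack and bees.
# Phase-shifted wing-flap and melee/pickaxe values are also computed, but no visible AnimGraph pin reads them.
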AnimBlueprint /BRCosmetics/Accessories/FORT_Backpacks/Backpack_F_MED_PollenTrove/Meshes/F_MED_PollenTrove_Pack_AnimBP
target_skeleton=&/F_MED_PollenTrove_Pack_Skeleton
graphs [
 AnimGraph {nodes [
   n1 = Control Rig!(Source=Input Pose(), Alpha, BeeMovement:Vector=Get Movement() -> (Movement:Vector)) -> (Pose) {
    alphaBoolBlend {blendOption=Linear}
    controlRigAssetReference {blueprintRigClass=&/F_MED_PollenTrove_Pack_CtrlRig.F_MED_PollenTrove_Pack_CtrlRig_C}
   }
   n2 = "Transform (Modify) Bone - Bone: pack"(ComponentPose=Local To Component(LocalPose=@n1), Translation="0,0.5,0", Rotation="0,0,0") {
    translationMode=BMM_Additive
    translationSpace=BCS_BoneSpace
    alphaBoolBlend {blendOption=Linear}
    boneToModify {boneName=pack}
    translation {y=0.5}
   }
   n3 = "Transform (Modify) Bone - Bone: pack"(ComponentPose=@n2, Translation=Get backpack_translation(), Rotation="0,0,0") {
    rotationMode=BMM_Additive
    rotationSpace=BCS_BoneSpace
    translationMode=BMM_Additive
    translationSpace=BCS_BoneSpace
    alphaBoolBlend {blendOption=Linear}
    boneToModify {boneName=pack}
   }
   n4 = "Transform (Modify) Bone - Bone: dyn_bee_01_wing_l"(ComponentPose=@n3, Alpha=1, Translation="0,0,0", Rotation="0,0,0", Scale="1,1,1") {rotationMode=BMM_Additive, rotationSpace=BCS_BoneSpace, alphaBoolBlend {blendOption=Linear}, boneToModify {boneName=dyn_bee_01_wing_l}}
   n5 = "Transform (Modify) Bone - Bone: dyn_bee_01_wing_r"(ComponentPose=@n4, Alpha=1, Translation="0,0,0", Rotation="0,0,0", Scale="1,1,1") {rotationMode=BMM_Additive, rotationSpace=BCS_BoneSpace, alphaBoolBlend {blendOption=Linear}, boneToModify {boneName=dyn_bee_01_wing_r}}
   n6 = "Transform (Modify) Bone - Bone: dyn_bee_02_wing_l"(ComponentPose=@n5, Alpha=1, Translation="0,0,0", Rotation="0,0,0", Scale="1,1,1") {rotationMode=BMM_Additive, rotationSpace=BCS_BoneSpace, alphaBoolBlend {blendOption=Linear}, boneToModify {boneName=dyn_bee_02_wing_l}}
   n7 = "Transform (Modify) Bone - Bone: dyn_bee_02_wing_r"(ComponentPose=@n6, Alpha=1, Translation="0,0,0", Rotation="0,0,0", Scale="1,1,1") {rotationMode=BMM_Additive, rotationSpace=BCS_BoneSpace, alphaBoolBlend {blendOption=Linear}, boneToModify {boneName=dyn_bee_02_wing_r}}
   n8 = "Transform (Modify) Bone - Bone: dyn_bee_03_wing_l"(ComponentPose=@n7, Alpha=1, Translation="0,0,0", Rotation="0,0,0", Scale="1,1,1") {rotationMode=BMM_Additive, rotationSpace=BCS_BoneSpace, alphaBoolBlend {blendOption=Linear}, boneToModify {boneName=dyn_bee_03_wing_l}}
   n9 = "Transform (Modify) Bone - Bone: dyn_bee_03_wing_r"(ComponentPose=@n8, Alpha=1, Translation="0,0,0", Rotation="0,0,0", Scale="1,1,1") {rotationMode=BMM_Additive, rotationSpace=BCS_BoneSpace, alphaBoolBlend {blendOption=Linear}, boneToModify {boneName=dyn_bee_03_wing_r}}
   n10 = "Transform (Modify) Bone - Bone: pack"(ComponentPose=@n9, Translation=Get backpack_translation(), Rotation=Get backpack_rotate()) {rotationMode=BMM_Additive, translationMode=BMM_Additive, alphaBoolBlend {blendOption=Linear}, boneToModify {boneName=pack}}
   n11 = "Transform (Modify) Bone - Bone: dyn_bee_02_body"(
    ComponentPose="Transform (Modify) Bone - Bone: dyn_bee_01_body"(ComponentPose=@n10, Alpha=1, Translation=Get FloatingMotion1() -> (FloatingMotion1:Vector), Rotation="0,0,0", Scale="1,1,1") {translationMode=BMM_Additive, alphaBoolBlend {blendOption=Linear}, boneToModify {boneName=dyn_bee_01_body}}
    Alpha=1
    Translation=Get FloatingMotion2() -> (FloatingMotion2:Vector)
    Rotation="0,0,0"
    Scale="1,1,1"
   ) {translationMode=BMM_Additive, alphaBoolBlend {blendOption=Linear}, boneToModify {boneName=dyn_bee_02_body}}
   n12 = Get FloatingMotion3() -> (FloatingMotion3:Vector)
   Output Pose(Result=Component To Local(ComponentPose="Transform (Modify) Bone - Bone: dyn_bee_03_body"(ComponentPose=@n11, Alpha=1, Translation=@n12, Rotation="0,0,0", Scale="1,1,1") {translationMode=BMM_Additive, alphaBoolBlend {blendOption=Linear}, boneToModify {boneName=dyn_bee_03_body}})) {layerGroup=None}
  ]}
 EventGraph {nodes [
   Event BlueprintUpdateAnimation()
   > n1 = Sequence!(execute) -> (then_0, then_1, then_2:Exec, then_3:Exec, then_4:Exec, then_5:Exec, then_6:Exec, then_7:Exec)
   n2 = GetRealTimeSeconds()
   n3 = "Sin (Radians)"(A=float * float(A=float * float(A=2, B=-1.2), B=@n2) {is=title K2Node_PromotableOperator})
   n4 = "Sin (Radians)"(A=float * float(A=float * float(A=1.4, B=-0.8), B=@n2))
   > Set FloatingMotion1(FloatingMotion1:Vector=MakeVector(X=float / float(A=@n3, B=-1.3), Y=0, Z=float / float(A=@n4, B=-1.2))) -> (Output_Get:Vector)
   n5 = GetRealTimeSeconds()
   n6 = "Sin (Radians)"(A=float * float(A=float * float(A=1, B=-1.6), B=@n5))
   n7 = "Sin (Radians)"(A=float * float(A=float * float(A=0.4, B=-0.8), B=@n5))
   Set FloatingMotion3(execute=@n1.then_2, FloatingMotion3:Vector=MakeVector(X=float / float(A=@n6, B=1.85), Y=0, Z=float / float(A=@n7, B=1.4))) -> (Output_Get:Vector)
   n8 = GetRealTimeSeconds()
   n9 = "Sin (Radians)"(A=float * float(A=float * float(A=2, B=-1.2), B=@n8))
   n10 = "Sin (Radians)"(A=float * float(A=float * float(A=2, B=2), B=@n8))
   n11 = "Sin (Radians)"(A=float * float(A=float * float(A=2.4, B=-0.8), B=@n8))
   Set FloatingMotion2(
    execute=@n1.then_1
    FloatingMotion2:Vector=MakeVector(X=float / float(A=@n9, B=-1.5), Y=float / float(A=@n10, B=4), Z=float / float(A=@n11, B=-1.3))
   ) -> (Output_Get:Vector)
   n12 = float * float(A=GetRealTimeSeconds(), B=50)
   n13 = float * float(A="Sin (Radians)"(A=float + float(A=@n12, B=1)), B=20)
   Set WingRot3(execute=@n1.then_5, WingRot3:Rotator=MakeRotator(Yaw=float + float(A=@n13, B=-10))) -> (Output_Get:Rotator)
   n14 = float * float(A=GetRealTimeSeconds(), B=50)
   n15 = float * float(A="Sin (Radians)"(A=float + float(A=@n14, B=0.5)), B=20)
   Set WingRot2(execute=@n1.then_4, WingRot2:Rotator=MakeRotator(Yaw=float + float(A=@n15, B=-10))) -> (Output_Get:Rotator)
   n16 = GetOwnerSkeletalMeshComponent()
   n17 = BreakRotator(InRot=BreakTransform(InTransform=GetDeltaTransformFromRefPose(self=@n16, BoneName=spine_05, BaseName=spine_01)).Rotation)
   n18 = MapRangeClamped(Value=@n17.Yaw, InRangeA=-10, InRangeB=-50, OutRangeB=50)
   Set backpack_rotate(execute=@n1.then_6, backpack_rotate=MakeRotator(Roll=@n18, Pitch=float / float(A=@n18, B=5)))
   > Set backpack_translation(backpack_translation=MakeVector(Y=MapRangeClamped(Value=@n17.Yaw, InRangeA=-15, InRangeB=-50, OutRangeB=-10)))
   n19 = GetOwnerSkeletalMeshComponent()
   > n20 = Set head down rot(head down rot:double=GetAlphaFromReferenceTransformDelta(SkeletalMeshComponent=@n19, BoneName=spine_01, BaseName=head, InputRange=(X=0,Y=-45), TwistAxis="1,0,0")) -> (Output_Get:double)
   n21 = "To Float (Boolean)"(InBool=Get bIsPlayingMeleeAnim())
   n22 = float * float(B=0.2)
   n23 = Get isMelee() -> (isMelee:double)
   > Set isMelee(isMelee:double=float * float(A=FInterpTo(Current=@n23, Target=@n21, DeltaTime=@n22, InterpSpeed=5), B=@n20)) -> (Output_Get:double)
   n24 = BreakRotator(InRot=BreakTransform(InTransform=GetDeltaTransformFromRefPose(self=@n19, BoneName=upperarm_l, BaseName=clavicle_l)).Rotation)
   > Set Backpack Rotate Pickaxe(Backpack Rotate Pickaxe:Rotator=MakeRotator(Roll=MapRangeClamped(Value=@n24.Yaw, InRangeA=0, InRangeB=90, OutRangeB=30))) -> (Output_Get:Rotator)
   > n25 = PlayerMovement(AnimInstance=Self-Reference(), XDotProductCoefficient=1, YDotProductCoefficient=1, ZDotProductCoefficient=1)
   > Set!(execute, Movement_X:double=@n25, Movement_Y:double=@n25.YMovement, Movement_Z:double=@n25.ZMovement) -> (then, Output_Get:Vector)
   n26 = GetRealTimeSeconds()
   n27 = float * float(A="Sin (Radians)"(A=float * float(A=@n26, B=50)), B=20)
   Set WingRot1(execute=@n1.then_3, WingRot1:Rotator=MakeRotator(Yaw=float + float(A=@n27, B=-10))) -> (Output_Get:Rotator)
   TryGetPawnOwner()
   n28 = float * float(A=1, B=1.1)
   float / float(A="Sin (Radians)"(A=float * float(A=@n28, B=@n2)), B=20)
   n29 = float * float(A=0.7, B=1.3)
   float / float(A="Sin (Radians)"(A=float * float(A=@n29, B=@n5)), B=4)
  ]}
]
variables [
 Backpack Rotate Pickaxe {type=Rotator}
 backpack_rotate {type=Rotator}
 backpack_translation {type=Vector}
 FloatingMotion1 {type=Vector}
 FloatingMotion2 {type=Vector}
 FloatingMotion3 {type=Vector}
 head down rot {type=number}
 isMelee {type=number}
 Movement {type=Vector}
 WingRot1 {type=Rotator}
 WingRot2 {type=Rotator}
 WingRot3 {type=Rotator}
]
